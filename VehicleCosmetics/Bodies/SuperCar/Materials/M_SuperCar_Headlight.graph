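# Material for the headlights of the SuperCar body in VehicleCosmetics, using the Default Lit shading model.
# Its only graph logic multiplies the EmissiveColor vector parameter, a pale cyan by default, by a fixed factor to boost brightness.
# The color can be overridden through the parameter, while the intensity multiplier is a constant.
# It is flagged for use on skeletal meshes and with mesh deformers.
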Material /VehicleCosmetics/Bodies/SuperCar/Materials/M_SuperCar_Headlight
graph {
 properties {bUsedWithMeshDeformer=true, bUsedWithSkeletalMesh=true, pixelDepthOffsetMode=PDOM_Legacy, usedShadingModels=Default Lit}
 nodes [Multiply(A=VectorParameter() {parameterName=EmissiveColor, defaultValue=<rgba 0.6146 1 0.9925 1>}.RGB) {constB=16}]
}
