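# Taillight material for the SuperCar vehicle body, using Default Lit shading and flagged for skeletal mesh and mesh deformer use.
# It multiplies a red EmissiveColor vector parameter by an IsBraking scalar parameter, so the brake value scales the intensity of the glow color.
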Material /VehicleCosmetics/Bodies/SuperCar/Materials/M_SuperCar_Taillight
graph {
 properties {bUsedWithMeshDeformer=true, bUsedWithSkeletalMesh=true, pixelDepthOffsetMode=PDOM_Legacy, usedShadingModels=Default Lit}
 nodes [Multiply(A=VectorParameter() {parameterName=EmissiveColor, defaultValue=<rgba 1 0 0.0167 1>}.RGB, B=ScalarParameter() {parameterName=IsBraking}) {constB=16}]
}
